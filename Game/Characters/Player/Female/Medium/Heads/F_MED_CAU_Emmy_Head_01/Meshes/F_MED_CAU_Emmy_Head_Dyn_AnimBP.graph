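# Post-style animation blueprint for the Emmy female player head that adds secondary physics motion to its hair bones.
# The input pose runs in component space through a serial chain of single-body Anim Dynamics nodes on dyn_hair bones 01–05 and 07–09.
# Most use angular springs in actor space; 07 and 08 use tighter, low-gravity limits with an angular bias override.
# An additive bone-space rotation offsets dyn_hair_02 mid-chain; the event graph has no logic.
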AnimBlueprint /Game/Characters/Player/Female/Medium/Heads/F_MED_CAU_Emmy_Head_01/Meshes/F_MED_CAU_Emmy_Head_Dyn_AnimBP
target_skeleton=/Game/Characters/Player/Common/Fortnite_Base_Head/Fortnite_Base_Head_Export_Skeleton
graphs [
 AnimGraph {nodes [
   AnimGraphNode_SubInput_0 = Input Pose()
   n1 = Anim Dynamics - Bone: dyn_hair_01(ComponentPose=Local To Component(LocalPose=@AnimGraphNode_SubInput_0), Alpha=0.5) {
    alpha=0.9
    angularSpringConstant=1500
    bAngularSpring=true
    bGravityOverrideInSimSpace=true
    gravityScale=0.65
    numSolverIterationsPostUpdate=4
    numSolverIterationsPreUpdate=16
    simulationSpace=Actor
    alphaBoolBlend {blendOption=Linear}
    boundBone {boneName=dyn_hair_01}
    physicsBodyDefinitions [{
      sphereCollisionRadius=0
      boundBone {boneName=dyn_hair_01}
      boxExtents=<15 15 2.5>
      constraintSetup {angularTargetAxis=AxisZ, twistAxis=AxisZ, angularLimitsMax=<15 15 15>, angularLimitsMin=<-15 -15 -15>, angularTarget=<0 0 1>}
      localJointOffset=<0 0 -2>
     }]
   }
   n2 = Anim Dynamics - Bone: dyn_hair_02(ComponentPose=@n1, Alpha=0.6) {
    alpha=0.9
    angularSpringConstant=1500
    bAngularSpring=true
    bGravityOverrideInSimSpace=true
    gravityScale=0.65
    numSolverIterationsPostUpdate=4
    numSolverIterationsPreUpdate=16
    simulationSpace=Actor
    alphaBoolBlend {blendOption=Linear}
    boundBone {boneName=dyn_hair_02}
    physicsBodyDefinitions [{
      sphereCollisionRadius=0
      boundBone {boneName=dyn_hair_02}
      boxExtents=<15 15 2.5>
      constraintSetup {angularTargetAxis=AxisZ, twistAxis=AxisZ, angularLimitsMax=<15 15 15>, angularLimitsMin=<-15 -15 -15>, angularTarget=<0 0 1>}
      localJointOffset=<0 0 -2>
     }]
   }
   n3 = "Transform (Modify) Bone - Bone: dyn_hair_02"!(ComponentPose=@n2, bAlphaBoolEnabled:bool=True, Translation="0,0,0", Rotation="35,12,21", Scale="1,1,1") -> (Pose) {
    alphaInputType=Bool
    rotationMode=BMM_Additive
    rotationSpace=BCS_BoneSpace
    scaleMode=BMM_Additive
    scaleSpace=BCS_BoneSpace
    translationMode=BMM_Additive
    translationSpace=BCS_BoneSpace
    alphaBoolBlend {blendOption=Linear}
    boneToModify {boneName=dyn_hair_02}
    rotation=<rot 35 12 21>
   }
   n4 = Anim Dynamics - Bone: dyn_hair_03(ComponentPose=@n3, Alpha=0.45, ExternalForce="0,0,0") {
    alpha=0.9
    angularSpringConstant=1500
    bAngularSpring=true
    bGravityOverrideInSimSpace=true
    gravityScale=0.65
    numSolverIterationsPostUpdate=4
    numSolverIterationsPreUpdate=16
    simulationSpace=Actor
    alphaBoolBlend {blendOption=Linear}
    boundBone {boneName=dyn_hair_03}
    physicsBodyDefinitions [{
      sphereCollisionRadius=0
      boundBone {boneName=dyn_hair_03}
      boxExtents=<15 15 2.5>
      constraintSetup {angularTargetAxis=AxisZ, twistAxis=AxisZ, angularLimitsMax=<15 15 15>, angularLimitsMin=<-15 -15 -15>, angularTarget=<0 0 1>}
      localJointOffset=<0 0 -2>
     }]
   }
   n5 = Anim Dynamics - Bone: dyn_hair_04(ComponentPose=@n4, Alpha=0.4, ExternalForce="0,0,0") {
    alpha=0.9
    angularSpringConstant=1500
    bAngularSpring=true
    bGravityOverrideInSimSpace=true
    gravityScale=0.65
    numSolverIterationsPostUpdate=4
    numSolverIterationsPreUpdate=16
    simulationSpace=Actor
    alphaBoolBlend {blendOption=Linear}
    boundBone {boneName=dyn_hair_04}
    physicsBodyDefinitions [{
      sphereCollisionRadius=0
      boundBone {boneName=dyn_hair_04}
      boxExtents=<15 15 2.5>
      constraintSetup {angularTargetAxis=AxisZ, twistAxis=AxisZ, angularLimitsMax=<15 15 15>, angularLimitsMin=<-15 -15 -15>, angularTarget=<0 0 1>}
      localJointOffset=<0 0 -2>
     }]
   }
   n6 = Anim Dynamics - Bone: dyn_hair_05(ComponentPose=@n5, Alpha=0.3, ExternalForce="0,0,0") {
    alpha=0.9
    angularSpringConstant=1500
    bAngularSpring=true
    bGravityOverrideInSimSpace=true
    gravityScale=0.65
    numSolverIterationsPostUpdate=4
    numSolverIterationsPreUpdate=16
    simulationSpace=Actor
    alphaBoolBlend {blendOption=Linear}
    boundBone {boneName=dyn_hair_05}
    physicsBodyDefinitions [{
      sphereCollisionRadius=0
      boundBone {boneName=dyn_hair_05}
      boxExtents=<15 15 2.5>
      constraintSetup {angularTargetAxis=AxisZ, twistAxis=AxisZ, angularLimitsMax=<15 15 15>, angularLimitsMin=<-15 -15 -15>, angularTarget=<0 0 1>}
      localJointOffset=<0 0 -2>
     }]
   }
   n7 = Anim Dynamics - Bone: dyn_hair_07(ComponentPose=@n6, Alpha=0.6, ExternalForce="0,0,0") {
    angularBiasOverride=1
    bGravityOverrideInSimSpace=true
    bOverrideAngularBias=true
    gravityScale=0.1
    alphaBoolBlend {blendOption=Linear}
    boundBone {boneName=dyn_hair_07}
    physicsBodyDefinitions [{
      sphereCollisionRadius=0
      boundBone {boneName=dyn_hair_07}
      boxExtents=<5 5 5>
      constraintSetup {
       angularTargetAxis=AxisY
       twistAxis=AxisY
       angularLimitsMax=<10 0 10>
       angularLimitsMin=<-10 0 -10>
       angularTarget=<0 0 0>
       linearAxesMax=<0 0 3>
      }
      localJointOffset=<0 0 -2>
     }]
   }
   n8 = Anim Dynamics - Bone: dyn_hair_08(ComponentPose=@n7, Alpha=0.7, ExternalForce="0,0,0") {
    angularBiasOverride=1
    bGravityOverrideInSimSpace=true
    bOverrideAngularBias=true
    gravityScale=0.1
    alphaBoolBlend {blendOption=Linear}
    boundBone {boneName=dyn_hair_08}
    physicsBodyDefinitions [{
      sphereCollisionRadius=0
      boundBone {boneName=dyn_hair_08}
      boxExtents=<5 5 5>
      constraintSetup {
       angularTargetAxis=AxisY
       twistAxis=AxisY
       angularLimitsMax=<10 0 10>
       angularLimitsMin=<-10 0 -10>
       angularTarget=<0 0 0>
       linearAxesMax=<0 0 3>
      }
      localJointOffset=<0 0 -2>
     }]
   }
   n9 = Anim Dynamics - Bone: dyn_hair_09(ComponentPose=@n8, Alpha=1, ExternalForce="0,0,0") {
    alpha=0.9
    angularSpringConstant=1500
    bAngularSpring=true
    bGravityOverrideInSimSpace=true
    numSolverIterationsPostUpdate=4
    numSolverIterationsPreUpdate=16
    simulationSpace=Actor
    alphaBoolBlend {blendOption=Linear}
    boundBone {boneName=dyn_hair_09}
    physicsBodyDefinitions [{
      sphereCollisionRadius=0
      boundBone {boneName=dyn_hair_09}
      boxExtents=<15 15 2.5>
      constraintSetup {angularTargetAxis=AxisZ, twistAxis=AxisY, angularLimitsMax=<15 15 15>, angularLimitsMin=<-15 -15 -15>, angularTarget=<0 0 1>}
      localJointOffset=<0 0 0>
     }]
   }
   Output Pose(Result=Component To Local(ComponentPose=@n9)) {layerGroup=None}
  ]}
 EventGraph {nodes [Event BlueprintUpdateAnimation()]}
]
variables []
